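# MetaSound patch for a weapon's stop-fire sound layer, outputting stereo audio and a finished trigger.
# Each gunshot starts a random wave from the supplied array with slight random pitch and gain variation.
# OnStop triggers an attack-decay envelope that scales the already-playing wave, with decay time set to twice the wave's duration, so the layer fades in when firing stops rather than starting then.
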MetaSoundPatch /FortMetaSoundWeapon/Items/Metasounds/Subgraphs/MSP_Weapon_Fire_OnStopFire
inputs [
 OnGunshot {type=Trigger}
 OnStop {type=Trigger}
 OnStopFire.Sounds {type=WaveAsset:Array}
]
outputs [
 OnFinished {type=Trigger}
 OnStopFire.L {type=Audio}
 OnStopFire.R {type=Audio}
]
pages [Default {
  input_defaults {OnGunshot=null, OnStop=null, OnStopFire.Sounds []}
  nodes [
   ArrayRandomGet_0 = Array.Random Get:WaveAsset:Array(
    Next=^OnGunshot
    In Array=^OnStopFire.Sounds
    No Repeats=UE.Divide:Int32(PrimaryOperand=Array.Num:WaveAsset:Array(Array=^OnStopFire.Sounds).Num).Out
   ) {also {defaults {Enable Shared State="<MetaSoundBuilderResult.SUCCEEDED: 0>", Weights []}}}
   GetWaveDuration_0 = "UE.Get Wave Duration: "(Wave=@ArrayRandomGet_0.Value)
   ADEnvelopeADEnvelope_0 = AD Envelope.AD Envelope:Float(
    Trigger=^OnStop
    Decay Time=UE.Multiply:Time by Float(PrimaryOperand=@GetWaveDuration_0.Duration) {also {defaults {AdditionalOperands=2}}}.Out
   ) {also {defaults {Attack Curve=2, Attack Time=0.1}}}
   n1 = UE.Multiply:Float(
    AdditionalOperands=@ADEnvelopeADEnvelope_0.Out Envelope
    PrimaryOperand=UE.RandomFloat(Next=^OnGunshot) {also {defaults {Max=1.05, Min=0.95}}}.Value
   )
   FrequencyMultipliertoSemitone_0 = UE.Frequency Multiplier to Semitone:Float(Frequency Multiplier=UE.RandomFloat(Next=^OnGunshot) {also {defaults {Max=1.05, Min=0.95}}}.Value)
   WavePlayer_0 = UE.Wave Player:Stereo(Play=^OnGunshot, Wave Asset=@ArrayRandomGet_0.Value, Pitch Shift=@FrequencyMultipliertoSemitone_0.Semitones)
   Multiply_0 = UE.Multiply:Audio by Float(AdditionalOperands=@n1.Out, PrimaryOperand=@WavePlayer_0.Out Right)
   Multiply_1 = UE.Multiply:Audio by Float(AdditionalOperands=@n1.Out, PrimaryOperand=@WavePlayer_0.Out Left)
  ]
  links [{dst="@OnStopFire.R", src=Multiply_0.Out}, {dst="@OnStopFire.L", src=Multiply_1.Out}, {dst="@OnFinished", src=WavePlayer_0.On Finished}]
 }]
