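# MetaSound patch that plays randomized mono gunshot sounds, named for third-person weapon fire, on each OnGunshot trigger.
# Each shot picks a random wave, pitch and volume, with a trigger sequence alternating two summed wave players unless bUseSingleVoice routes every shot to the first.
# Trigger Controls pass OnFinished only from the most recently started player.
# A Mobile page reduces this to one wave player with the same randomization.
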MetaSoundPatch /FortMetaSoundWeapon/Items/Metasounds/Subgraphs/MSP_Weapon_Fire_3p_Mono
inputs [
 bUseSingleVoice {type=Bool}
 OnGunshot {type=Trigger}
 Random.Pitch.Max {type=Float}
 Random.Pitch.Min {type=Float}
 Random.Volume.Max {type=Float}
 Random.Volume.Min {type=Float}
 Sounds {type=WaveAsset:Array}
]
outputs [
 OnFinished {type=Trigger}
 Out.L {type=Audio}
]
pages [
 Default {
  is_build_graph=true
  input_defaults {
   OnGunshot=null
   Random.Pitch.Max=0
   Random.Pitch.Min=0
   Random.Volume.Max=0
   Random.Volume.Min=0
   Sounds=null
   bUseSingleVoice="<MetaSoundBuilderResult.SUCCEEDED: 0>"
  }
  nodes [
   n1 = TriggerCompare.Clamp:Bool(Compare=^OnGunshot, A=^bUseSingleVoice) {also {defaults {B="<MetaSoundBuilderResult.SUCCEEDED: 0>"}}}
   TriggerSequence2_0 = "UE.Trigger Sequence (2)"(In=@n1.False)
   TriggerAnyTriggerAny2_0 = "TriggerAny.Trigger Any (2)"(In 0=@n1.True, In 1=@TriggerSequence2_0.Out 0)
   n2 = VariableAccessor.Trigger(Variable=VariableMutator.Trigger(Variable=InitVariable.Trigger().Variable, Value=@TriggerAnyTriggerAny2_0.Out).Variable)
   ArrayRandomGet_0 = Array.Random Get:WaveAsset:Array(Next=^OnGunshot, In Array=^Sounds) {also {defaults {Enable Shared State="<MetaSoundBuilderResult.SUCCEEDED: 0>", No Repeats=-1}}}
   FrequencyMultipliertoSemitone_0 = UE.Frequency Multiplier to Semitone:Float(Frequency Multiplier=UE.RandomFloat(Next=^OnGunshot, Max=^Random.Pitch.Max, Min=^Random.Pitch.Min).Value)
   WavePlayer_0 = UE.Wave Player:Mono(Wave Asset=@ArrayRandomGet_0.Value, Pitch Shift=UE.Reroute(Value=@FrequencyMultipliertoSemitone_0.Semitones).Value, Play=@n2.Value) {also {version [1,1]}}
   Multiply_0 = UE.Multiply:Audio by Float(
    AdditionalOperands=UE.RandomFloat(Max=^Random.Volume.Max, Min=^Random.Volume.Min, Next=VariableAccessor.Trigger(Variable=@n2.Variable).Value).Value
    PrimaryOperand=@WavePlayer_0.Out Mono
   )
   n3 = VariableAccessor.Trigger(Variable=VariableMutator.Trigger(Variable=InitVariable.Trigger().Variable, Value=@TriggerSequence2_0.Out 1).Variable)
   n4 = VariableAccessor.Trigger(Variable=@n3.Variable)
   WavePlayer_1 = UE.Wave Player:Mono(
    Wave Asset=UE.Reroute(Value=@ArrayRandomGet_0.Value).Value
    Pitch Shift=UE.Reroute(Value=@FrequencyMultipliertoSemitone_0.Semitones).Value
    Play=@n3.Value
   ) {also {version [1,1]}}
   Add_0 = UE.Add:Audio(
    PrimaryOperand=@Multiply_0.Out
    AdditionalOperands=UE.Multiply:Audio by Float(
     AdditionalOperands=UE.RandomFloat(Max=^Random.Volume.Max, Min=^Random.Volume.Min, Next=@n4.Value).Value
     PrimaryOperand=@WavePlayer_1.Out Mono
    ).Out
   )
   TriggerControl_0 = UE.Trigger Control(Trigger In=@WavePlayer_0.On Finished, Open=@WavePlayer_0.On Play, Close=@WavePlayer_1.On Play)
   TriggerControl_1 = UE.Trigger Control(Close=@WavePlayer_0.On Play, Trigger In=@WavePlayer_1.On Finished, Open=@WavePlayer_1.On Play)
   TriggerAnyTriggerAny2_1 = "TriggerAny.Trigger Any (2)"(In 0=@TriggerControl_0.Trigger Out, In 1=@TriggerControl_1.Trigger Out)
  ]
  links [{dst="@Out.L", src=Add_0.Out}, {dst="@OnFinished", src=TriggerAnyTriggerAny2_1.Out}]
 }
 Mobile {
  nodes [
   InitVariable.Trigger()
   InitVariable.Trigger()
   ArrayRandomGet_0 = Array.Random Get:WaveAsset:Array(Next=^OnGunshot, In Array=^Sounds) {also {defaults {Enable Shared State="<MetaSoundBuilderResult.SUCCEEDED: 0>", No Repeats=-1}}}
   FrequencyMultipliertoSemitone_0 = UE.Frequency Multiplier to Semitone:Float(Frequency Multiplier=UE.RandomFloat(Next=^OnGunshot, Max=^Random.Pitch.Max, Min=^Random.Pitch.Min).Value)
   WavePlayer_0 = UE.Wave Player:Mono(Play=^OnGunshot, Wave Asset=@ArrayRandomGet_0.Value, Pitch Shift=@FrequencyMultipliertoSemitone_0.Semitones) {also {version [1,1]}}
   Multiply_0 = UE.Multiply:Audio by Float(
    AdditionalOperands=UE.RandomFloat(Next=^OnGunshot, Max=^Random.Volume.Max, Min=^Random.Volume.Min).Value
    PrimaryOperand=@WavePlayer_0.Out Mono
   )
  ]
  links [{dst="@Out.L", src=Multiply_0.Out}, {dst="@OnFinished", src=WavePlayer_0.On Finished}]
 }
]
