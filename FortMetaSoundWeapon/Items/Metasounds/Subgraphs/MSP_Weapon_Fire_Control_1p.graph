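# MetaSound patch for first-person weapon fire control that turns automatic, burst and single-shot inputs into gunshot triggers.
# bIsFiring runs a repeat at the fire rate, while FireBurst runs a counted repeat behind a gate that deferred trigger variables close, then reopen after a fire-rate delay.
# Repeated shots pass an ammo check, using a counter reset on AmmoRemaining changes, unless ammo is unlimited; single shots bypass it.
# It also outputs the computed current ammo.
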MetaSoundPatch /FortMetaSoundWeapon/Items/Metasounds/Subgraphs/MSP_Weapon_Fire_Control_1p
inputs [
 bHasUnlimitedAmmo {type=Bool}
 bIsFiring {type=Bool}
 BurstRate {type=Float}
 FireBurst {type=Trigger}
 FireSingleShot {type=Trigger}
 OnPlay {type=Trigger}
 RoundsPerBurst {type=Int32}
 StopBurst {type=Trigger}
 Weapon.AmmoRemaining {type=Int32}
 Weapon.FireRate {type=Float}
]
outputs [
 CurrentAmmo {type=Int32}
 OutGunshot {type=Trigger}
]
pages [Default {
  input_defaults {
   BurstRate=0
   FireBurst=null
   FireSingleShot=null
   OnPlay=null
   RoundsPerBurst=2
   StopBurst=null
   Weapon.AmmoRemaining=30
   Weapon.FireRate=0.5
   bHasUnlimitedAmmo=null
   bIsFiring=null
  }
  nodes [
   InitVariable.Int32()
   InitVariable.Trigger()
   InitVariable.Trigger()
   InitVariable.Trigger()
   InitVariable.Trigger()
   n1 = InitVariable.Trigger()
   n2 = VariableDeferredAccessor.Trigger(Variable=InitVariable.Trigger().Variable)
   n3 = VariableDeferredAccessor.Trigger(Variable=InitVariable.Trigger().Variable)
   TriggerControl_0 = UE.Trigger Control(Trigger In=^FireBurst, Close=@n2.Value, Open=@n3.Value)
   TriggerOnValueChange_0 = UE.Trigger On Value Change:Bool(Value=^bIsFiring)
   TriggerAnyTriggerAny2_0 = "TriggerAny.Trigger Any (2)"(In 0=^OnPlay, In 1=@TriggerOnValueChange_0.Trigger)
   n4 = TriggerCompare.Clamp:Bool(A=^bIsFiring, Compare=@TriggerAnyTriggerAny2_0.Out) {also {defaults {B="<MetaSoundBuilderResult.SUCCEEDED: 0>"}}}
   TriggerAnyTriggerAny2_2 = "TriggerAny.Trigger Any (2)"(
    In 1=UE.TriggerRepeat(Num Repeats=^RoundsPerBurst, Stop=^StopBurst, Period=UE.ConversionFloatToTime(In=^BurstRate).Out, Start=@TriggerControl_0.Trigger Out).RepeatOut
    In 0=UE.TriggerRepeat(Period=UE.ConversionFloatToTime(In=^Weapon.FireRate).Out, Stop=@n4.False, Start=@n4.True).RepeatOut
   )
   n5 = TriggerCompare.Clamp:Bool(A=^bHasUnlimitedAmmo, Compare=@TriggerAnyTriggerAny2_2.Out)
   TriggerOnValueChange_1 = UE.Trigger On Value Change:Int32(Value=^Weapon.AmmoRemaining)
   TriggerCounter_0 = UE.Trigger Counter(In=@n5.True, Reset=@TriggerOnValueChange_1.Trigger)
   Subtract_0 = UE.Subtract:Int32(
    PrimaryOperand=UE.Add:Int32(PrimaryOperand=^Weapon.AmmoRemaining) {also {defaults {AdditionalOperands=1}}}.Out
    AdditionalOperands=@TriggerCounter_0.Count
   )
   TriggerAnyTriggerAny2_1 = "TriggerAny.Trigger Any (2)"(
    In 1=UE.Reroute(Value=@n5.False).Value
    In 0=TriggerCompare.Clamp:Int32(A=@Subtract_0.Out, Compare=@TriggerCounter_0.On Trigger) {also {defaults {B=0, Type=5}}}.True
   )
   TriggerAnyTriggerAny2_3 = "TriggerAny.Trigger Any (2)"(
    In 0=^FireSingleShot
    In 1=VariableAccessor.Trigger(Variable=VariableMutator.Trigger(Variable=@n1.Variable, Value=@TriggerAnyTriggerAny2_1.Out).Variable).Value
   )
   VariableMutator.Trigger(Value=@TriggerControl_0.Trigger Out, Variable=@n2.Variable)
   TriggerDelay_0 = UE.Trigger Delay(Delay Time=UE.ConversionFloatToTime(In=^Weapon.FireRate).Out, In=@TriggerControl_0.Trigger Out)
   VariableMutator.Trigger(Value=@TriggerDelay_0.Out, Variable=@n3.Variable)
  ]
  links [{dst="@CurrentAmmo", src=Subtract_0.Out}, {dst="@OutGunshot", src=TriggerAnyTriggerAny2_3.Out}]
 }]
