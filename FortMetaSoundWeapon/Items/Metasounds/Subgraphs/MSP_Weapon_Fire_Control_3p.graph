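# MetaSound patch for third-person weapon fire control that turns fire inputs into gunshot triggers and a predicted ammo count.
# Repeaters drive automatic fire and bursts, with a delayed gate limiting burst restarts to the fire rate.
# Repeated shots are checked against a counter that resets when reported ammo changes, with a bypass for unlimited ammo.
# Single shots and resolved start/stop triggers skip the ammo check.
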MetaSoundPatch /FortMetaSoundWeapon/Items/Metasounds/Subgraphs/MSP_Weapon_Fire_Control_3p
inputs [
 bHasUnlimitedAmmo {type=Bool}
 BurstRate {type=Float}
 FireBurst {type=Trigger}
 FireSingleShot {type=Trigger}
 RoundsPerBurst {type=Int32}
 StartFiring {type=Trigger}
 StopBurst {type=Trigger}
 StopFiring {type=Trigger}
 Weapon.AmmoRemaining {type=Int32}
 Weapon.FireRate {type=Float}
]
outputs [
 CurrentAmmo {type=Int32}
 OutGunshot {type=Trigger}
]
pages [Default {
  input_defaults {
   BurstRate=0
   FireBurst=null
   FireSingleShot=null
   RoundsPerBurst=2
   StartFiring=null
   StopBurst=null
   StopFiring=null
   Weapon.AmmoRemaining=30
   Weapon.FireRate=0.5
   bHasUnlimitedAmmo=null
  }
  nodes [
   n1 = InitVariable.Trigger()
   n2 = VariableDeferredAccessor.Trigger(Variable=InitVariable.Trigger().Variable)
   n3 = VariableDeferredAccessor.Trigger(Variable=InitVariable.Trigger().Variable)
   TriggerControl_0 = UE.Trigger Control(Trigger In=^FireBurst, Open=@n2.Value, Close=@n3.Value)
   n4 = &MSP_TriggerResolve(A=^StartFiring, B=^StopFiring)
   TriggerAnyTriggerAny2_0 = "TriggerAny.Trigger Any (2)"(
    In 1=UE.TriggerRepeat(Num Repeats=^RoundsPerBurst, Stop=^StopBurst, Period=UE.ConversionFloatToTime(In=^BurstRate).Out, Start=@TriggerControl_0.Trigger Out).RepeatOut
    In 0=UE.TriggerRepeat(Period=UE.ConversionFloatToTime(In=^Weapon.FireRate).Out, Start=@n4.A, Stop=@n4.B).RepeatOut
   )
   n5 = TriggerCompare.Clamp:Bool(A=^bHasUnlimitedAmmo, Compare=@TriggerAnyTriggerAny2_0.Out)
   TriggerOnValueChange_0 = UE.Trigger On Value Change:Int32(Value=^Weapon.AmmoRemaining)
   TriggerCounter_0 = UE.Trigger Counter(In=@n5.True, Reset=@TriggerOnValueChange_0.Trigger)
   Subtract_0 = UE.Subtract:Int32(
    PrimaryOperand=UE.Add:Int32(PrimaryOperand=^Weapon.AmmoRemaining) {also {defaults {AdditionalOperands=1}}}.Out
    AdditionalOperands=@TriggerCounter_0.Count
   )
   TriggerAnyTriggerAny2_1 = "TriggerAny.Trigger Any (2)"(
    In 1=UE.Reroute(Value=@n5.False).Value
    In 0=TriggerCompare.Clamp:Int32(A=@Subtract_0.Out, Compare=@TriggerCounter_0.On Trigger) {also {defaults {B=0, Type=5}}}.True
   )
   n6 = InitVariable.Trigger()
   TriggerAnyTriggerAny3_0 = "TriggerAny.Trigger Any (3)"(
    In 0=^FireSingleShot
    In 2=VariableAccessor.Trigger(Variable=VariableMutator.Trigger(Variable=@n1.Variable, Value=@TriggerAnyTriggerAny2_1.Out).Variable).Value
    In 1=VariableAccessor.Trigger(Variable=VariableMutator.Trigger(Variable=@n6.Variable, Value=@n4.OnResolved).Variable).Value
   )
   TriggerDelay_0 = UE.Trigger Delay(Delay Time=UE.ConversionFloatToTime(In=^Weapon.FireRate).Out, In=@TriggerControl_0.Trigger Out)
   VariableMutator.Trigger(Value=@TriggerDelay_0.Out, Variable=@n2.Variable)
   VariableMutator.Trigger(Value=@TriggerControl_0.Trigger Out, Variable=@n3.Variable)
  ]
  links [{dst="@CurrentAmmo", src=Subtract_0.Out}, {dst="@OutGunshot", src=TriggerAnyTriggerAny3_0.Out}]
 }]
