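# MetaSound patch that generates a randomized weapon-fire variation profile of volume range, pitch range and EQ settings, seeded by a character's unique ID.
# On Generate, paired random floats feed Min and Max nodes to form ordered volume and pitch ranges within the input bounds.
# Single random floats set EQ gain and cutoff from input ranges, and bandwidth from a hardcoded range.
# All random nodes share the same seed input.
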MetaSoundPatch /FortMetaSoundWeapon/Items/Metasounds/Subgraphs/MSP_Weapon_Fire_GenerateProfile
inputs [
 Character.UniqueID {type=Int32}
 Generate {type=Trigger}
 Random.EQ.Frequency.Max {type=Float}
 Random.EQ.Frequency.Min {type=Float}
 Random.EQ.Gain.Max {type=Float}
 Random.EQ.Gain.Min {type=Float}
 Random.Pitch.Max {type=Float}
 Random.Pitch.Min {type=Float}
 Random.Volume.Max {type=Float}
 Random.Volume.Min {type=Float}
]
outputs [
 EQ.Bandwidth {type=Float}
 EQ.CutoffFrequency {type=Float}
 EQ.Gain {type=Float}
 Pitch.Max {type=Float}
 Pitch.Min {type=Float}
 Volume.Max {type=Float}
 Volume.Min {type=Float}
]
pages [Default {
  input_defaults {
   Character.UniqueID=-1
   Generate=null
   Random.EQ.Frequency.Max=20000
   Random.EQ.Frequency.Min=10000
   Random.EQ.Gain.Max=0
   Random.EQ.Gain.Min=-1
   Random.Pitch.Max=1.15
   Random.Pitch.Min=0.85
   Random.Volume.Max=1.15
   Random.Volume.Min=0.85
  }
  nodes [
   InitVariable.Float()
   InitVariable.Int32()
   n1 = UE.RandomFloat(Seed=^Character.UniqueID, Next=^Generate, Max=^Random.Volume.Max, Min=^Random.Volume.Min)
   n2 = UE.RandomFloat(Seed=^Character.UniqueID, Next=^Generate, Max=^Random.Volume.Max, Min=^Random.Volume.Min)
   MaxMax_0 = Max.Max:Float(A=@n1.Value, B=@n2.Value)
   n3 = UE.RandomFloat(Seed=^Character.UniqueID, Next=^Generate, Max=^Random.Pitch.Max, Min=^Random.Pitch.Min)
   n4 = UE.RandomFloat(Seed=^Character.UniqueID, Next=^Generate, Max=^Random.Pitch.Max, Min=^Random.Pitch.Min)
   MaxMax_1 = Max.Max:Float(A=@n3.Value, B=@n4.Value)
   MinMin_0 = Min.Min:Float(A=@n1.Value, B=@n2.Value)
   MinMin_1 = Min.Min:Float(A=@n3.Value, B=@n4.Value)
   RandomFloat_2 = UE.RandomFloat(Seed=^Character.UniqueID, Next=^Generate, Max=^Random.EQ.Gain.Max, Min=^Random.EQ.Gain.Min)
   RandomFloat_5 = UE.RandomFloat(Seed=^Character.UniqueID, Next=^Generate, Max=^Random.EQ.Frequency.Max, Min=^Random.EQ.Frequency.Min)
   RandomFloat_6 = UE.RandomFloat(Seed=^Character.UniqueID, Next=^Generate) {also {defaults {Max=1.2, Min=0.1}}}
  ]
  links [
   {dst="@Volume.Max", src=MaxMax_0.Value}
   {dst="@Pitch.Max", src=MaxMax_1.Value}
   {dst="@Volume.Min", src=MinMin_0.Value}
   {dst="@Pitch.Min", src=MinMin_1.Value}
   {dst="@EQ.Gain", src=RandomFloat_2.Value}
   {dst="@EQ.CutoffFrequency", src=RandomFloat_5.Value}
   {dst="@EQ.Bandwidth", src=RandomFloat_6.Value}
  ]
 }]
