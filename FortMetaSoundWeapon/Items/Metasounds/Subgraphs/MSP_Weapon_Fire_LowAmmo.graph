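# MetaSound patch for a weapon fire's low-ammo layer, outputting mono audio and a finished trigger.
# On each gunshot at or below the ammo threshold, it plays a random pitch-varied sound that grows louder as the clip empties.
# A nested threshold adds a last-shot sound through a self-closing trigger gate, and aiming down sights scales the mix.
# The Mobile page keeps only the low-ammo player, without last-shot or ADS handling.
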MetaSoundPatch /FortMetaSoundWeapon/Items/Metasounds/Subgraphs/MSP_Weapon_Fire_LowAmmo
inputs [
 CurrentAmmo {type=Int32}
 LastShotOfClip.LastShotThreshold {type=Int32}
 LastShotOfClip.Sounds {type=WaveAsset:Array}
 LastShotOfClip.Volume.RandomVolumeMax {type=Float}
 LastShotOfClip.Volume.RandomVolumeMin {type=Float}
 LastShotOfClip.Volume.Volume {type=Float}
 LowAmmo.ADS.VolumeMultiply {type=Float}
 LowAmmo.Random.Pitch.Max {type=Float}
 LowAmmo.Random.Pitch.Min {type=Float}
 LowAmmo.Random.Volume.Max {type=Float}
 LowAmmo.Random.Volume.Min {type=Float}
 LowAmmo.Sounds {type=WaveAsset:Array}
 LowAmmo.Threshold {type=Int32}
 OnGunshot {type=Trigger}
 Weapon.IsAimingDownSights {type=Bool}
]
outputs [
 LowAmmo.C {type=Audio}
 OnFinished {type=Trigger}
]
pages [
 Default {
  is_build_graph=true
  input_defaults {
   CurrentAmmo=0
   LastShotOfClip.LastShotThreshold=1
   LastShotOfClip.Volume.RandomVolumeMax=1
   LastShotOfClip.Volume.RandomVolumeMin=1
   LastShotOfClip.Volume.Volume=1
   LowAmmo.ADS.VolumeMultiply=1
   LowAmmo.Random.Pitch.Max=1.05
   LowAmmo.Random.Pitch.Min=0.95
   LowAmmo.Random.Volume.Max=1.05
   LowAmmo.Random.Volume.Min=0.95
   LowAmmo.Threshold=0
   OnGunshot=null
   Weapon.IsAimingDownSights=null
   LastShotOfClip.Sounds []
   LowAmmo.Sounds []
  }
  nodes [
   n1 = Convert.Float:Bool(Bool=^Weapon.IsAimingDownSights)
   n2 = TriggerCompare.Clamp:Int32(A=^CurrentAmmo, B=^LowAmmo.Threshold, Compare=^OnGunshot) {also {defaults {Type=4}}}
   n3 = TriggerCompare.Clamp:Int32(A=^CurrentAmmo, B=^LastShotOfClip.LastShotThreshold, Compare=@n2.True) {also {defaults {Type=4}}}
   n4 = VariableDeferredAccessor.Trigger(Variable=InitVariable.Trigger().Variable)
   TriggerControl_0 = UE.Trigger Control(Open=@n3.False, Trigger In=@n3.True, Close=@n4.Value)
   n5 = VariableAccessor.Trigger(Variable=VariableMutator.Trigger(Value=@TriggerControl_0.Trigger Out, Variable=@n4.Variable).Variable)
   n6 = VariableAccessor.Trigger(Variable=@n5.Variable)
   n7 = VariableAccessor.Trigger(Variable=@n6.Variable)
   n8 = Convert.Float:Int32(Int32=^LowAmmo.Threshold)
   n9 = Convert.Float:Int32(Int32=^CurrentAmmo)
   n10 = VariableAccessor.Trigger(Variable=VariableMutator.Trigger(Variable=InitVariable.Trigger().Variable, Value=@n2.True).Variable)
   n11 = VariableAccessor.Trigger(Variable=@n10.Variable)
   n12 = VariableAccessor.Trigger(Variable=@n11.Variable)
   ArrayRandomGet_0 = Array.Random Get:WaveAsset:Array(In Array=^LowAmmo.Sounds, Next=@n12.Value) {also {defaults {Enable Shared State="<MetaSoundBuilderResult.SUCCEEDED: 0>", Weights []}}}
   FrequencyMultipliertoSemitone_0 = UE.Frequency Multiplier to Semitone:Float(Frequency Multiplier=UE.RandomFloat(Max=^LowAmmo.Random.Pitch.Max, Min=^LowAmmo.Random.Pitch.Min, Next=@n10.Value).Value)
   WavePlayer_0 = UE.Wave Player:Mono(
    Wave Asset=@ArrayRandomGet_0.Value
    Pitch Shift=@FrequencyMultipliertoSemitone_0.Semitones
    Play=VariableAccessor.Trigger(Variable=@n12.Variable).Value
   ) {also {version [1,1]}}
   ArrayRandomGet_1 = Array.Random Get:WaveAsset:Array(In Array=^LastShotOfClip.Sounds, Next=@n5.Value) {also {defaults {Enable Shared State="<MetaSoundBuilderResult.SUCCEEDED: 0>", Weights []}}}
   WavePlayer_1 = UE.Wave Player:Mono(Wave Asset=@ArrayRandomGet_1.Value, Pitch Shift=@FrequencyMultipliertoSemitone_0.Semitones, Play=@n6.Value) {also {version [1,1]}}
   AudioMixerAudioMixerMono2_0 = "AudioMixer.Audio Mixer (Mono, 2)"(
    Gain 0=UE.Multiply:Float(
     AdditionalOperands=^LastShotOfClip.Volume.Volume
     PrimaryOperand=UE.RandomFloat(Max=^LastShotOfClip.Volume.RandomVolumeMax, Min=^LastShotOfClip.Volume.RandomVolumeMin, Next=@n7.Value).Value
    ).Out
    Gain 1=UE.Multiply:Float(
     PrimaryOperand=MapRange.MapRange:Float(In Range A=@n8.Float, In=@n9.Float) {also {defaults {In Range B=0, Out Range A=0.3}}}.Out Value
     AdditionalOperands=UE.RandomFloat(Max=^LowAmmo.Random.Volume.Max, Min=^LowAmmo.Random.Volume.Min, Next=@n11.Value).Value
    ).Out
    In 1=@WavePlayer_0.Out Mono
    In 0=@WavePlayer_1.Out Mono
   )
   Multiply_0 = UE.Multiply:Audio by Float(
    AdditionalOperands=UE.Add:Float(
     PrimaryOperand=UE.Multiply:Float(PrimaryOperand=^LowAmmo.ADS.VolumeMultiply, AdditionalOperands=@n1.Float).Out
     AdditionalOperands=UE.Subtract:Float(AdditionalOperands=@n1.Float) {also {defaults {PrimaryOperand=1}}}.Out
    ).Out
    PrimaryOperand=@AudioMixerAudioMixerMono2_0.Out
   )
   TriggerAnyTriggerAny2_1 = "TriggerAny.Trigger Any (2)"(In 0=UE.Reroute(Value=@n3.False).Value, In 1=@WavePlayer_1.On Finished)
   TriggerAccumulatorTriggerAccumulate2_0 = "TriggerAccumulator.Trigger Accumulate (2)"(In 0=@TriggerAnyTriggerAny2_1.Out, In 1=@WavePlayer_0.On Finished) {also {defaults {Auto Reset="<MetaSoundBuilderResult.SUCCEEDED: 0>"}}}
   TriggerAnyTriggerAny2_0 = "TriggerAny.Trigger Any (2)"(In 1=@TriggerAccumulatorTriggerAccumulate2_0.Out, In 0=@n2.False)
  ]
  links [{dst="@LowAmmo.C", src=Multiply_0.Out}, {dst="@OnFinished", src=TriggerAnyTriggerAny2_0.Out}]
 }
 Mobile {
  nodes [
   n1 = Convert.Float:Int32(Int32=^LowAmmo.Threshold)
   n2 = Convert.Float:Int32(Int32=^CurrentAmmo)
   n3 = TriggerCompare.Clamp:Int32(A=^CurrentAmmo, B=^LowAmmo.Threshold, Compare=^OnGunshot) {also {defaults {Type=4}}}
   n4 = VariableAccessor.Trigger(Variable=VariableMutator.Trigger(Variable=InitVariable.Trigger().Variable, Value=@n3.True).Variable)
   n5 = VariableAccessor.Trigger(Variable=@n4.Variable)
   n6 = Array.Num:WaveAsset:Array(Array=^LowAmmo.Sounds)
   n7 = VariableAccessor.Trigger(Variable=@n5.Variable)
   ArrayRandomGet_0 = Array.Random Get:WaveAsset:Array(
    In Array=^LowAmmo.Sounds
    No Repeats=Max.Max:Int32(A=UE.Subtract:Int32(PrimaryOperand=@n6.Num) {also {defaults {AdditionalOperands=2}}}.Out) {also {defaults {B=1}}}.Value
    Next=@n7.Value
   ) {also {defaults {Enable Shared State="<MetaSoundBuilderResult.SUCCEEDED: 0>", Weights []}}}
   FrequencyMultipliertoSemitone_0 = UE.Frequency Multiplier to Semitone:Float(Frequency Multiplier=UE.RandomFloat(Max=^LowAmmo.Random.Pitch.Max, Min=^LowAmmo.Random.Pitch.Min, Next=@n4.Value).Value)
   WavePlayer_0 = UE.Wave Player:Mono(
    Wave Asset=@ArrayRandomGet_0.Value
    Pitch Shift=@FrequencyMultipliertoSemitone_0.Semitones
    Play=VariableAccessor.Trigger(Variable=@n7.Variable).Value
   ) {also {version [1,1]}}
   Multiply_0 = UE.Multiply:Audio by Float(
    AdditionalOperands=UE.Multiply:Float(
     AdditionalOperands=MapRange.MapRange:Float(In Range A=@n1.Float, In=@n2.Float) {also {defaults {In Range B=0, Out Range A=0.3}}}.Out Value
     PrimaryOperand=UE.RandomFloat(Max=^LowAmmo.Random.Volume.Max, Min=^LowAmmo.Random.Volume.Min, Next=@n5.Value).Value
    ).Out
    PrimaryOperand=@WavePlayer_0.Out Mono
   )
   TriggerAnyTriggerAny2_0 = "TriggerAny.Trigger Any (2)"(In 0=@n3.False, In 1=@WavePlayer_0.On Finished)
  ]
  links [{dst="@LowAmmo.C", src=Multiply_0.Out}, {dst="@OnFinished", src=TriggerAnyTriggerAny2_0.Out}]
 }
]
