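# MetaSound patch for the main layer of a weapon gunshot, playing a random non-repeating wave with randomized pitch and volume on OnGunshot.
# The default page alternates two 5.1 wave players, or uses one with bUseSingleVoice, applies an aim-down-sights volume multiplier and mixes each channel.
# Trigger gates let only the most recently started voice fire OnFinished.
# The Mobile page uses one stereo player without ADS scaling.
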MetaSoundPatch /FortMetaSoundWeapon/Items/Metasounds/Subgraphs/MSP_Weapon_Fire_MainFire
inputs [
 bUseSingleVoice {type=Bool}
 MainFire.ADS.VolumeMultiply {type=Float}
 MainFire.Random.Pitch.Max {type=Float}
 MainFire.Random.Pitch.Min {type=Float}
 MainFire.Random.Volume.Max {type=Float}
 MainFire.Random.Volume.Min {type=Float}
 MainFire.Sounds {type=WaveAsset:Array}
 OnGunshot {type=Trigger}
 Weapon.IsAimingDownSights {type=Bool}
]
outputs [
 Duration {type=Time}
 MainFire.C {type=Audio}
 MainFire.L {type=Audio}
 MainFire.LFE {type=Audio}
 MainFire.Ls {type=Audio}
 MainFire.R {type=Audio}
 MainFire.Rs {type=Audio}
 OnFinished {type=Trigger}
]
pages [
 Default {
  is_build_graph=true
  input_defaults {
   MainFire.ADS.VolumeMultiply=1
   MainFire.Random.Pitch.Max=1.05
   MainFire.Random.Pitch.Min=0.95
   MainFire.Random.Volume.Max=1.05
   MainFire.Random.Volume.Min=0.95
   OnGunshot=null
   Weapon.IsAimingDownSights=null
   bUseSingleVoice=null
   MainFire.Sounds []
  }
  nodes [
   n1 = Convert.Float:Bool(Bool=^Weapon.IsAimingDownSights)
   n2 = TriggerCompare.Clamp:Bool(Compare=^OnGunshot, A=^bUseSingleVoice) {also {defaults {B="<MetaSoundBuilderResult.SUCCEEDED: 0>"}}}
   TriggerSequence2_0 = "UE.Trigger Sequence (2)"(In=@n2.False)
   n3 = VariableAccessor.Trigger(Variable=VariableMutator.Trigger(Variable=InitVariable.Trigger().Variable, Value=@TriggerSequence2_0.Out 1).Variable)
   Multiply_1 = UE.Multiply:Float(
    AdditionalOperands=UE.Add:Float(
     PrimaryOperand=UE.Multiply:Float(PrimaryOperand=^MainFire.ADS.VolumeMultiply, AdditionalOperands=@n1.Float).Out
     AdditionalOperands=UE.Subtract:Float(AdditionalOperands=@n1.Float) {also {defaults {PrimaryOperand=1}}}.Out
    ).Out
    PrimaryOperand=UE.RandomFloat(Max=^MainFire.Random.Volume.Max, Min=^MainFire.Random.Volume.Min, Next=VariableAccessor.Trigger(Variable=@n3.Variable).Value).Value
   )
   n4 = Convert.Float:Bool(Bool=^Weapon.IsAimingDownSights)
   Multiply_0 = UE.Multiply:Float(PrimaryOperand=^MainFire.ADS.VolumeMultiply, AdditionalOperands=@n4.Float)
   TriggerAnyTriggerAny2_0 = "TriggerAny.Trigger Any (2)"(In 0=@n2.True, In 1=@TriggerSequence2_0.Out 0)
   n5 = VariableAccessor.Trigger(Variable=VariableMutator.Trigger(Variable=InitVariable.Trigger().Variable, Value=@TriggerAnyTriggerAny2_0.Out).Variable)
   n6 = UE.Multiply:Float(
    AdditionalOperands=UE.Add:Float(PrimaryOperand=@Multiply_0.Out, AdditionalOperands=UE.Subtract:Float(AdditionalOperands=@n4.Float) {also {defaults {PrimaryOperand=1}}}.Out).Out
    PrimaryOperand=UE.RandomFloat(Max=^MainFire.Random.Volume.Max, Min=^MainFire.Random.Volume.Min, Next=VariableAccessor.Trigger(Variable=@n5.Variable).Value).Value
   )
   n7 = UE.RandomFloat(Max=^MainFire.Random.Pitch.Max, Min=^MainFire.Random.Pitch.Min, Next=^OnGunshot)
   FrequencyMultipliertoSemitone_0 = UE.Frequency Multiplier to Semitone:Float(Frequency Multiplier=@n7.Value)
   n8 = Array.Num:WaveAsset:Array(Array=^MainFire.Sounds)
   ArrayRandomGet_0 = Array.Random Get:WaveAsset:Array(
    In Array=^MainFire.Sounds
    Next=^OnGunshot
    No Repeats=Max.Max:Int32(A=UE.Subtract:Int32(PrimaryOperand=@n8.Num) {also {defaults {AdditionalOperands=2}}}.Out) {also {defaults {B=1}}}.Value
   ) {also {defaults {Enable Shared State="<MetaSoundBuilderResult.SUCCEEDED: 0>", Weights []}}}
   WavePlayer_0 = UE.Wave Player:5dot1(Pitch Shift=@FrequencyMultipliertoSemitone_0.Semitones, Wave Asset=UE.Reroute(Value=@ArrayRandomGet_0.Value).Value, Play=@n3.Value)
   FrequencyMultipliertoSemitone_1 = UE.Frequency Multiplier to Semitone:Float(Frequency Multiplier=@n7.Value)
   WavePlayer_1 = UE.Wave Player:5dot1(Wave Asset=@ArrayRandomGet_0.Value, Pitch Shift=@FrequencyMultipliertoSemitone_1.Semitones, Play=@n5.Value)
   AudioMixerAudioMixerMono2_0 = "AudioMixer.Audio Mixer (Mono, 2)"(Gain 1=@Multiply_1.Out, Gain 0=@n6.Out, In 1=@WavePlayer_0.Out Low Frequency, In 0=@WavePlayer_1.Out Low Frequency)
   AudioMixerAudioMixerMono2_1 = "AudioMixer.Audio Mixer (Mono, 2)"(Gain 1=@Multiply_1.Out, Gain 0=@n6.Out, In 1=@WavePlayer_0.Out Side Left, In 0=@WavePlayer_1.Out Side Left)
   AudioMixerAudioMixerMono2_2 = "AudioMixer.Audio Mixer (Mono, 2)"(Gain 1=@Multiply_1.Out, Gain 0=@n6.Out, In 1=@WavePlayer_0.Out Front Center, In 0=@WavePlayer_1.Out Front Center)
   AudioMixerAudioMixerMono2_3 = "AudioMixer.Audio Mixer (Mono, 2)"(Gain 1=@Multiply_1.Out, Gain 0=@n6.Out, In 1=@WavePlayer_0.Out Side Right, In 0=@WavePlayer_1.Out Side Right)
   AudioMixerAudioMixerMono2_4 = "AudioMixer.Audio Mixer (Mono, 2)"(Gain 1=@Multiply_1.Out, Gain 0=@n6.Out, In 1=@WavePlayer_0.Out Front Right, In 0=@WavePlayer_1.Out Front Right)
   AudioMixerAudioMixerMono2_5 = "AudioMixer.Audio Mixer (Mono, 2)"(Gain 1=@Multiply_1.Out, Gain 0=@n6.Out, In 1=@WavePlayer_0.Out Front Left, In 0=@WavePlayer_1.Out Front Left)
   GetWaveDuration_0 = "UE.Get Wave Duration: "(Wave=@ArrayRandomGet_0.Value)
   TriggerControl_0 = UE.Trigger Control(Close=@WavePlayer_0.On Play, Trigger In=@WavePlayer_1.On Finished, Open=@WavePlayer_1.On Play)
   TriggerControl_1 = UE.Trigger Control(Trigger In=@WavePlayer_0.On Finished, Open=@WavePlayer_0.On Play, Close=@WavePlayer_1.On Play)
   TriggerAnyTriggerAny2_1 = "TriggerAny.Trigger Any (2)"(In 0=@TriggerControl_0.Trigger Out, In 1=@TriggerControl_1.Trigger Out)
  ]
  links [
   {dst="@MainFire.LFE", src=AudioMixerAudioMixerMono2_0.Out}
   {dst="@MainFire.Ls", src=AudioMixerAudioMixerMono2_1.Out}
   {dst="@MainFire.C", src=AudioMixerAudioMixerMono2_2.Out}
   {dst="@MainFire.Rs", src=AudioMixerAudioMixerMono2_3.Out}
   {dst="@MainFire.R", src=AudioMixerAudioMixerMono2_4.Out}
   {dst="@MainFire.L", src=AudioMixerAudioMixerMono2_5.Out}
   {dst="@Duration", src=GetWaveDuration_0.Duration}
   {dst="@OnFinished", src=TriggerAnyTriggerAny2_1.Out}
  ]
 }
 Mobile {
  nodes [
   n1 = Array.Num:WaveAsset:Array(Array=^MainFire.Sounds)
   ArrayRandomGet_0 = Array.Random Get:WaveAsset:Array(
    In Array=^MainFire.Sounds
    Next=^OnGunshot
    No Repeats=Max.Max:Int32(A=UE.Subtract:Int32(PrimaryOperand=@n1.Num) {also {defaults {AdditionalOperands=2}}}.Out) {also {defaults {B=1}}}.Value
   ) {also {defaults {Enable Shared State="<MetaSoundBuilderResult.SUCCEEDED: 0>", Weights []}}}
   GetWaveDuration_0 = "UE.Get Wave Duration: "(Wave=@ArrayRandomGet_0.Value)
   InitVariable.Trigger()
   InitVariable.Trigger()
   n2 = UE.RandomFloat(Max=^MainFire.Random.Volume.Max, Min=^MainFire.Random.Volume.Min, Next=UE.Reroute(Value=@ArrayRandomGet_0.On Next).Value)
   FrequencyMultipliertoSemitone_0 = UE.Frequency Multiplier to Semitone:Float(Frequency Multiplier=UE.RandomFloat(Max=^MainFire.Random.Pitch.Max, Min=^MainFire.Random.Pitch.Min, Next=^OnGunshot).Value)
   WavePlayer_0 = UE.Wave Player:Stereo(Play=@ArrayRandomGet_0.On Next, Wave Asset=@ArrayRandomGet_0.Value, Pitch Shift=@FrequencyMultipliertoSemitone_0.Semitones)
   Multiply_0 = UE.Multiply:Audio by Float(AdditionalOperands=@n2.Value, PrimaryOperand=@WavePlayer_0.Out Left)
   Multiply_1 = UE.Multiply:Audio by Float(AdditionalOperands=@n2.Value, PrimaryOperand=@WavePlayer_0.Out Right)
  ]
  links [
   {dst="@Duration", src=GetWaveDuration_0.Duration}
   {dst="@MainFire.L", src=Multiply_0.Out}
   {dst="@MainFire.R", src=Multiply_1.Out}
   {dst="@OnFinished", src=WavePlayer_0.On Finished}
  ]
 }
]
