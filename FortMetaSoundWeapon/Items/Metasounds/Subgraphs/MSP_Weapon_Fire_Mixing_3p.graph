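# MetaSound patch that mixes third-person weapon fire from close, medium and distant layers into stereo.
# Distance curves weight the layers, with optional close-layer EQ, followed by azimuth-driven targeting, high-pass and profile filters.
# Affiliation-based modulator volume and a fire-rate envelope scale the signal, which switches to an enveloped suppressor mix when a non-permanent suppressor is enabled.
# The Mobile page drops the filters and fire-rate envelope.
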MetaSoundPatch /FortMetaSoundWeapon/Items/Metasounds/Subgraphs/MSP_Weapon_Fire_Mixing_3p
inputs [
 bEnableDistanceEQ {type=Bool}
 Character.Affiliation {type=Enum:FortAudioAffiliation}
 Close.L {type=Audio}
 Close.R {type=Audio}
 CurveEvaluation.Close {type=Float}
 CurveEvaluation.Distant {type=Float}
 CurveEvaluation.Medium {type=Float}
 Distant.L {type=Audio}
 FireSingleShot {type=Trigger}
 Medium.L {type=Audio}
 Mixing.Friendly {type=Float}
 OnGunshot {type=Trigger}
 Orientation.Azimuth {type=Float}
 ProfileFilter.Bandwidth {type=Float}
 ProfileFilter.CutoffFrequency {type=Float}
 ProfileFilter.Gain {type=Float}
 Spatialization.Elevation {type=Float}
 Suppressor.Close.L {type=Audio}
 Suppressor.Close.R {type=Audio}
 Suppressor.EnvelopeDecay {type=Time}
 Suppressor.Medium {type=Audio}
 Suppressor.Mixing.BaseWeaponVolume {type=Float}
 Suppressor.Mixing.SuppressorVolume {type=Float}
 Targeting.FilterBandwidth {type=Float}
 Targeting.FilterFrequency {type=Float}
 Targeting.FilterGain {type=Float}
 Volume.Close {type=Float}
 Volume.Distant {type=Float}
 Volume.Medium {type=Float}
 Weapon.FireRate {type=Float}
 Weapon.IsAlwaysSuppressed {type=Bool}
 Weapon.IsSuppressorEnabled {type=Bool}
]
outputs [
 Out.L {type=Audio}
 Out.R {type=Audio}
]
pages [
 Default {
  is_build_graph=true
  input_defaults {
   Character.Affiliation=0
   Close.L=null
   Close.R=null
   CurveEvaluation.Close=0
   CurveEvaluation.Distant=0
   CurveEvaluation.Medium=0
   Distant.L=null
   FireSingleShot=null
   Medium.L=null
   Mixing.Friendly=0
   OnGunshot=null
   Orientation.Azimuth=180
   ProfileFilter.Bandwidth=0
   ProfileFilter.CutoffFrequency=0
   ProfileFilter.Gain=0
   Spatialization.Elevation=0
   Suppressor.Close.L=null
   Suppressor.Close.R=null
   Suppressor.EnvelopeDecay=0
   Suppressor.Medium=null
   Suppressor.Mixing.BaseWeaponVolume=0
   Suppressor.Mixing.SuppressorVolume=0
   Targeting.FilterBandwidth=4
   Targeting.FilterFrequency=300
   Targeting.FilterGain=-6
   Volume.Close=0.644
   Volume.Distant=0.6
   Volume.Medium=0.7
   Weapon.FireRate=1
   Weapon.IsAlwaysSuppressed=null
   Weapon.IsSuppressorEnabled=null
   bEnableDistanceEQ="<MetaSoundBuilderResult.SUCCEEDED: 0>"
  }
  nodes [
   ADEnvelopeADEnvelope_0 = AD Envelope.AD Envelope:Float(Trigger=^OnGunshot, Decay Time=^Suppressor.EnvelopeDecay) {also {defaults {Attack Curve=2, Decay Curve=0.5, Hard Reset="<MetaSoundBuilderResult.SUCCEEDED: 0>"}}}
   n1 = UE.Multiply:Float(AdditionalOperands=Modulation.GetModulatorValue() {also {defaults {Modulator=/FortMetaSoundWeapon/Items/Modulation/CB_WeaponFire_3p_Friendly_Vol}}}.Out)
   CrossfadeTriggerRouteFloat4_0 = "Crossfade.Trigger Route (Float, 4):Float"(
    Crossfade Value=Convert.Float:Int32(Int32=Convert.Int32:Enum:FortAudioAffiliation("Enum:FortAudioAffiliation"=^Character.Affiliation).Int32).Float
    In 2=Modulation.GetModulatorValue() {also {defaults {Modulator=/FortMetaSoundWeapon/Items/Modulation/CB_WeaponFire_3p_Enemy_Vol}}}.Out
    In 1=@n1.Out
    In 3=@n1.Out
   ) {also {defaults {In 0=1}}}
   n2 = UE.Multiply:Float(PrimaryOperand=^Weapon.FireRate) {also {defaults {AdditionalOperands=15}}}
   ADEnvelopeADEnvelope_1 = AD Envelope.AD Envelope:Float(
    Trigger=^FireSingleShot
    Decay Time=UE.ConversionFloatToTime(In=UE.Multiply:Float(PrimaryOperand=@n2.Out) {also {defaults {AdditionalOperands=1.2}}}.Out).Out
   ) {also {defaults {Attack Curve=0.4, Attack Time=0, Hard Reset="<MetaSoundBuilderResult.SUCCEEDED: 0>"}}}
   n3 = UE.Multiply:Float(
    AdditionalOperands=@CrossfadeTriggerRouteFloat4_0.Out
    PrimaryOperand=MapRange.MapRange:Float(In=@ADEnvelopeADEnvelope_1.Out Envelope) {also {defaults {Out Range A=0.8}}}.Out Value
   )
   n4 = UE.Multiply:Float(
    AdditionalOperands=@ADEnvelopeADEnvelope_0.Out Envelope
    PrimaryOperand=UE.Multiply:Float(
     PrimaryOperand=UE.Reroute(Value=@n3.Out).Value
     AdditionalOperands=UE.Subtract:Float(AdditionalOperands=^CurveEvaluation.Distant) {also {defaults {PrimaryOperand=1}}}.Out
    ).Out
   )
   n5 = AbsoluteValue.Abs:Float(Input=MapRange.MapRange:Float(In=Clamp.Clamp:Float(In=^Orientation.Azimuth) {also {defaults {Max=180, Min=0}}}.Value) {also {defaults {In Range B=180, Out Range A=-1}}}.Out Value)
   n6 = TriggerCompare.Clamp:Float(Compare=^OnGunshot, A=@n5.Value) {also {defaults {B=0.25, Type=5}}}
   n7 = UE.Multiply:Float(
    PrimaryOperand=UE.Multiply:Float(
     PrimaryOperand=@n5.Value
     AdditionalOperands=Value.Value:Float(Target Value=^Targeting.FilterGain, Reset=@n6.False, Set=@n6.True).Output Value
    ).Out
    AdditionalOperands=UE.Subtract:Float(AdditionalOperands=^CurveEvaluation.Distant) {also {defaults {PrimaryOperand=1}}}.Out
   )
   n8 = UE.Add:Audio(
    PrimaryOperand=UE.Multiply:Audio by Float(
     PrimaryOperand=^Medium.L
     AdditionalOperands=UE.Multiply:Float(PrimaryOperand=^CurveEvaluation.Medium, AdditionalOperands=^Volume.Medium).Out
    ).Out
    AdditionalOperands=UE.Multiply:Audio by Float(
     PrimaryOperand=^Distant.L
     AdditionalOperands=UE.Multiply:Float(PrimaryOperand=^CurveEvaluation.Distant, AdditionalOperands=^Volume.Distant).Out
    ).Out
   )
   n9 = UE.Multiply:Audio by Float(PrimaryOperand=^Close.L, AdditionalOperands=^CurveEvaluation.Close)
   n10 = UE.Subtract:Float(AdditionalOperands=Convert.Float:Bool(Bool=^bEnableDistanceEQ).Float) {also {defaults {PrimaryOperand=1}}}
   n11 = Convert.Float:Bool(Bool=^bEnableDistanceEQ)
   n12 = MapRange.MapRange:Float(In=^CurveEvaluation.Close) {also {defaults {In Range A=0.95, In Range B=0.2, Out Range A=20000, Out Range B=16000}}}
   n13 = MapRange.MapRange:Float(In=^CurveEvaluation.Close) {also {defaults {In Range A=0.95, In Range B=0.2, Out Range B=100}}}
   OnePoleHighPassFilter_1 = UE.One-Pole High Pass Filter:Audio(Cutoff Frequency=@n13.Out Value, In=@n9.Out)
   OnePoleLowPassFilter_0 = UE.One-Pole Low Pass Filter:Audio(Cutoff Frequency=@n12.Out Value, In=@OnePoleHighPassFilter_1.Out)
   n14 = UE.Multiply:Audio by Float(
    AdditionalOperands=^Volume.Close
    PrimaryOperand=UE.Add:Audio(
     AdditionalOperands=UE.Multiply:Audio by Float(PrimaryOperand=@n9.Out, AdditionalOperands=@n10.Out).Out
     PrimaryOperand=UE.Multiply:Audio by Float(AdditionalOperands=@n11.Float, PrimaryOperand=@OnePoleLowPassFilter_0.Out).Out
    ).Out
   )
   BiquadFilter_3 = UE.Biquad Filter:Audio(
    Bandwidth=^Targeting.FilterBandwidth
    Cutoff Frequency=^Targeting.FilterFrequency
    Gain=@n7.Out
    In=UE.Reroute(Value=UE.Add:Audio(AdditionalOperands=@n8.Out, PrimaryOperand=@n14.Out).Out).Value
   ) {also {defaults {Type=4}, version [2,0]}}
   OnePoleHighPassFilter_2 = UE.One-Pole High Pass Filter:Audio(In=@BiquadFilter_3.Out)
   BiquadFilter_1 = UE.Biquad Filter:Audio(
    Bandwidth=^ProfileFilter.Bandwidth
    Cutoff Frequency=^ProfileFilter.CutoffFrequency
    Gain=^ProfileFilter.Gain
    In=@OnePoleHighPassFilter_2.Out
   ) {also {defaults {Type=4}, version [2,0]}}
   n15 = UE.Multiply:Audio by Float(PrimaryOperand=^Close.R, AdditionalOperands=^CurveEvaluation.Close)
   OnePoleHighPassFilter_3 = UE.One-Pole High Pass Filter:Audio(Cutoff Frequency=@n13.Out Value, In=@n15.Out)
   OnePoleLowPassFilter_1 = UE.One-Pole Low Pass Filter:Audio(Cutoff Frequency=@n12.Out Value, In=@OnePoleHighPassFilter_3.Out)
   n16 = UE.Multiply:Audio by Float(
    AdditionalOperands=^Volume.Close
    PrimaryOperand=UE.Add:Audio(
     PrimaryOperand=UE.Multiply:Audio by Float(AdditionalOperands=@n11.Float, PrimaryOperand=@OnePoleLowPassFilter_1.Out).Out
     AdditionalOperands=UE.Multiply:Audio by Float(PrimaryOperand=@n15.Out, AdditionalOperands=@n10.Out).Out
    ).Out
   )
   BiquadFilter_0 = UE.Biquad Filter:Audio(
    Bandwidth=^Targeting.FilterBandwidth
    Cutoff Frequency=^Targeting.FilterFrequency
    Gain=@n7.Out
    In=UE.Reroute(Value=UE.Add:Audio(AdditionalOperands=@n8.Out, PrimaryOperand=@n16.Out).Out).Value
   ) {also {defaults {Type=4}, version [2,0]}}
   OnePoleHighPassFilter_0 = UE.One-Pole High Pass Filter:Audio(
    In=@BiquadFilter_0.Out
    Cutoff Frequency=MapRange.MapRange:Float(In=^Spatialization.Elevation) {also {defaults {In Range A=-90, In Range B=90, Out Range B=100}}}.Out Value
   )
   BiquadFilter_2 = UE.Biquad Filter:Audio(
    Bandwidth=^ProfileFilter.Bandwidth
    Cutoff Frequency=^ProfileFilter.CutoffFrequency
    Gain=^ProfileFilter.Gain
    In=@OnePoleHighPassFilter_0.Out
   ) {also {defaults {Type=4}, version [2,0]}}
   AudioMixerAudioMixerStereo3_0 = "AudioMixer.Audio Mixer (Stereo, 3)"(
    In 0 L=^Suppressor.Close.L
    In 0 R=^Suppressor.Close.R
    In 1 L=^Suppressor.Medium
    In 1 R=^Suppressor.Medium
    Gain 2=^Suppressor.Mixing.BaseWeaponVolume
    In 2 L=UE.Multiply:Audio by Float(AdditionalOperands=@n4.Out, PrimaryOperand=UE.Reroute(Value=@BiquadFilter_1.Out).Value).Out
    Gain 1=UE.Multiply:Float(PrimaryOperand=^CurveEvaluation.Medium, AdditionalOperands=^Suppressor.Mixing.SuppressorVolume).Out
    In 2 R=UE.Multiply:Audio by Float(AdditionalOperands=@n4.Out, PrimaryOperand=UE.Reroute(Value=@BiquadFilter_2.Out).Value).Out
    Gain 0=UE.Multiply:Float(PrimaryOperand=^CurveEvaluation.Close, AdditionalOperands=^Suppressor.Mixing.SuppressorVolume).Out
   )
   n17 = UE.Multiply:Float(
    PrimaryOperand=Convert.Float:Bool(Bool=^Weapon.IsSuppressorEnabled).Float
    AdditionalOperands=UE.Subtract:Float(AdditionalOperands=Convert.Float:Bool(Bool=^Weapon.IsAlwaysSuppressed).Float) {also {defaults {PrimaryOperand=1}}}.Out
   )
   AudioMixerAudioMixerStereo2_0 = "AudioMixer.Audio Mixer (Stereo, 2)"(
    In 1 L=@AudioMixerAudioMixerStereo3_0.Out L
    In 1 R=@AudioMixerAudioMixerStereo3_0.Out R
    In 0 L=UE.Multiply:Audio by Float(PrimaryOperand=@BiquadFilter_1.Out, AdditionalOperands=@n3.Out).Out
    Gain 1=@n17.Out
    In 0 R=UE.Multiply:Audio by Float(PrimaryOperand=@BiquadFilter_2.Out, AdditionalOperands=@n3.Out).Out
    Gain 0=UE.Subtract:Float(AdditionalOperands=@n17.Out) {also {defaults {PrimaryOperand=1}}}.Out
   )
  ]
  links [{dst="@Out.L", src=AudioMixerAudioMixerStereo2_0.Out L}, {dst="@Out.R", src=AudioMixerAudioMixerStereo2_0.Out R}]
 }
 Mobile {
  nodes [
   n1 = UE.Add:Audio(
    AdditionalOperands=UE.Multiply:Audio by Float(
     PrimaryOperand=^Distant.L
     AdditionalOperands=UE.Multiply:Float(PrimaryOperand=^CurveEvaluation.Distant, AdditionalOperands=^Volume.Distant).Out
    ).Out
    PrimaryOperand=UE.Multiply:Audio by Float(
     PrimaryOperand=^Medium.L
     AdditionalOperands=UE.Multiply:Float(PrimaryOperand=^CurveEvaluation.Medium, AdditionalOperands=^Volume.Medium).Out
    ).Out
   )
   n2 = UE.Add:Audio(
    AdditionalOperands=@n1.Out
    PrimaryOperand=UE.Multiply:Audio by Float(PrimaryOperand=^Close.R, AdditionalOperands=UE.Multiply:Float(PrimaryOperand=^CurveEvaluation.Close, AdditionalOperands=^Volume.Close).Out).Out
   )
   ADEnvelopeADEnvelope_0 = AD Envelope.AD Envelope:Float(Trigger=^OnGunshot, Decay Time=^Suppressor.EnvelopeDecay) {also {defaults {Attack Curve=2, Decay Curve=0.5, Hard Reset="<MetaSoundBuilderResult.SUCCEEDED: 0>"}}}
   n3 = UE.Multiply:Float(AdditionalOperands=Modulation.GetModulatorValue() {also {defaults {Modulator=&CB_WeaponFire_3p_Friendly_Vol}}}.Out)
   CrossfadeTriggerRouteFloat4_0 = "Crossfade.Trigger Route (Float, 4):Float"(
    Crossfade Value=Convert.Float:Int32(Int32=Convert.Int32:Enum:FortAudioAffiliation("Enum:FortAudioAffiliation"=^Character.Affiliation).Int32).Float
    In 2=Modulation.GetModulatorValue() {also {defaults {Modulator=&CB_WeaponFire_3p_Enemy_Vol}}}.Out
    In 1=@n3.Out
    In 3=@n3.Out
   ) {also {defaults {In 0=1}}}
   n4 = UE.Multiply:Float(
    AdditionalOperands=@ADEnvelopeADEnvelope_0.Out Envelope
    PrimaryOperand=UE.Multiply:Float(
     PrimaryOperand=UE.Reroute(Value=@CrossfadeTriggerRouteFloat4_0.Out).Value
     AdditionalOperands=UE.Subtract:Float(AdditionalOperands=^CurveEvaluation.Distant) {also {defaults {PrimaryOperand=1}}}.Out
    ).Out
   )
   AudioMixerAudioMixerStereo3_0 = "AudioMixer.Audio Mixer (Stereo, 3)"(
    In 0 L=^Suppressor.Close.L
    In 0 R=^Suppressor.Close.R
    In 1 L=^Suppressor.Medium
    In 1 R=^Suppressor.Medium
    Gain 2=^Suppressor.Mixing.BaseWeaponVolume
    In 2 R=UE.Multiply:Audio by Float(PrimaryOperand=@n2.Out, AdditionalOperands=@n4.Out).Out
    Gain 1=UE.Multiply:Float(PrimaryOperand=^CurveEvaluation.Medium, AdditionalOperands=^Suppressor.Mixing.SuppressorVolume).Out
    Gain 0=UE.Multiply:Float(PrimaryOperand=^CurveEvaluation.Close, AdditionalOperands=^Suppressor.Mixing.SuppressorVolume).Out
    In 2 L=UE.Multiply:Audio by Float(PrimaryOperand=@n2.Out, AdditionalOperands=@n4.Out).Out
   )
   n5 = UE.Multiply:Float(
    PrimaryOperand=Convert.Float:Bool(Bool=^Weapon.IsSuppressorEnabled).Float
    AdditionalOperands=UE.Subtract:Float(AdditionalOperands=Convert.Float:Bool(Bool=^Weapon.IsAlwaysSuppressed).Float) {also {defaults {PrimaryOperand=1}}}.Out
   )
   n6 = UE.Add:Audio(
    AdditionalOperands=@n1.Out
    PrimaryOperand=UE.Multiply:Audio by Float(PrimaryOperand=^Close.L, AdditionalOperands=UE.Multiply:Float(PrimaryOperand=^CurveEvaluation.Close, AdditionalOperands=^Volume.Close).Out).Out
   )
   AudioMixerAudioMixerStereo2_0 = "AudioMixer.Audio Mixer (Stereo, 2)"(
    In 1 L=@AudioMixerAudioMixerStereo3_0.Out L
    In 1 R=@AudioMixerAudioMixerStereo3_0.Out R
    Gain 1=@n5.Out
    In 0 R=UE.Multiply:Audio by Float(PrimaryOperand=@n6.Out, AdditionalOperands=@CrossfadeTriggerRouteFloat4_0.Out).Out
    In 0 L=UE.Multiply:Audio by Float(PrimaryOperand=@n6.Out, AdditionalOperands=@CrossfadeTriggerRouteFloat4_0.Out).Out
    Gain 0=UE.Subtract:Float(AdditionalOperands=@n5.Out) {also {defaults {PrimaryOperand=1}}}.Out
   )
  ]
  links [{dst="@Out.L", src=AudioMixerAudioMixerStereo2_0.Out L}, {dst="@Out.R", src=AudioMixerAudioMixerStereo2_0.Out R}]
 }
]
