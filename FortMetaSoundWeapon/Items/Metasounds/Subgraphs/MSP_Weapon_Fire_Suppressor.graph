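# MetaSound patch that plays the stereo suppressor layer of weapon fire, gated on each gunshot by the suppressor-enabled input.
# The default page alternates shots between two wave players to allow overlap, or uses one when bUseSingleVoice is set, then mixes them.
# Waves come from a shared-state random pick with a no-repeat count scaled to array size, with randomized pitch and volume.
# The Mobile page always uses a single player.
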MetaSoundPatch /FortMetaSoundWeapon/Items/Metasounds/Subgraphs/MSP_Weapon_Fire_Suppressor
inputs [
 bUseSingleVoice {type=Bool}
 OnGunshot {type=Trigger}
 Random.Pitch.Max {type=Float}
 Random.Pitch.Min {type=Float}
 Random.Volume.Max {type=Float}
 Random.Volume.Min {type=Float}
 Sounds {type=WaveAsset:Array}
 Weapon.IsSuppressorEnabled {type=Bool}
]
outputs [
 OutL {type=Audio}
 OutR {type=Audio}
]
pages [
 Default {
  is_build_graph=true
  input_defaults {
   OnGunshot=null
   Random.Pitch.Max=1.03
   Random.Pitch.Min=0.98
   Random.Volume.Max=1.05
   Random.Volume.Min=0.95
   Weapon.IsSuppressorEnabled=null
   bUseSingleVoice="<MetaSoundBuilderResult.SUCCEEDED: 0>"
   Sounds []
  }
  nodes [
   n1 = TriggerCompare.Clamp:Bool(
    A=^bUseSingleVoice
    Compare=TriggerCompare.Clamp:Bool(Compare=^OnGunshot, A=^Weapon.IsSuppressorEnabled) {also {defaults {B="<MetaSoundBuilderResult.SUCCEEDED: 0>"}}}.True
   ) {also {defaults {B="<MetaSoundBuilderResult.SUCCEEDED: 0>"}}}
   TriggerSequence2_0 = "UE.Trigger Sequence (2)"(In=@n1.False)
   n2 = VariableAccessor.Trigger(Variable=VariableMutator.Trigger(Variable=InitVariable.Trigger().Variable, Value=@TriggerSequence2_0.Out 1).Variable)
   n3 = VariableAccessor.Trigger(Variable=@n2.Variable)
   TriggerAnyTriggerAny2_1 = "TriggerAny.Trigger Any (2)"(In 0=@n1.True, In 1=@TriggerSequence2_0.Out 0)
   n4 = VariableAccessor.Trigger(Variable=VariableMutator.Trigger(Variable=InitVariable.Trigger().Variable, Value=@TriggerAnyTriggerAny2_1.Out).Variable)
   n5 = VariableAccessor.Trigger(Variable=@n4.Variable)
   n6 = VariableAccessor.Trigger(Variable=@n5.Variable)
   n7 = VariableAccessor.Trigger(Variable=@n3.Variable)
   TriggerAnyTriggerAny2_0 = "TriggerAny.Trigger Any (2)"(In 0=VariableAccessor.Trigger(Variable=@n6.Variable).Value, In 1=VariableAccessor.Trigger(Variable=@n7.Variable).Value)
   n8 = UE.RandomFloat(Max=^Random.Pitch.Max, Min=^Random.Pitch.Min, Next=@TriggerAnyTriggerAny2_0.Out)
   FrequencyMultipliertoSemitone_1 = UE.Frequency Multiplier to Semitone:Float(Frequency Multiplier=@n8.Value)
   n9 = Array.Num:WaveAsset:Array(Array=^Sounds)
   TriggerAnyTriggerAny2_2 = "TriggerAny.Trigger Any (2)"(In 0=@n6.Value, In 1=@n7.Value)
   ArrayRandomGet_0 = Array.Random Get:WaveAsset:Array(
    In Array=^Sounds
    No Repeats=Max.Max:Int32(A=UE.Subtract:Int32(PrimaryOperand=@n9.Num) {also {defaults {AdditionalOperands=2}}}.Out) {also {defaults {B=1}}}.Value
    Next=@TriggerAnyTriggerAny2_2.Out
   ) {also {defaults {Enable Shared State="<MetaSoundBuilderResult.SUCCEEDED: 0>", Shared State Behavior=2, Weights []}}}
   WavePlayer_0 = UE.Wave Player:Stereo(Pitch Shift=@FrequencyMultipliertoSemitone_1.Semitones, Wave Asset=UE.Reroute(Value=@ArrayRandomGet_0.Value).Value, Play=@n2.Value)
   FrequencyMultipliertoSemitone_0 = UE.Frequency Multiplier to Semitone:Float(Frequency Multiplier=@n8.Value)
   WavePlayer_1 = UE.Wave Player:Stereo(Wave Asset=@ArrayRandomGet_0.Value, Pitch Shift=@FrequencyMultipliertoSemitone_0.Semitones, Play=@n4.Value)
   AudioMixerAudioMixerStereo2_0 = "AudioMixer.Audio Mixer (Stereo, 2)"(
    Gain 1=UE.RandomFloat(Max=^Random.Volume.Max, Min=^Random.Volume.Min, Next=@n3.Value).Value
    Gain 0=UE.RandomFloat(Max=^Random.Volume.Max, Min=^Random.Volume.Min, Next=@n5.Value).Value
    In 1 L=@WavePlayer_0.Out Left
    In 1 R=@WavePlayer_0.Out Right
    In 0 L=@WavePlayer_1.Out Left
    In 0 R=@WavePlayer_1.Out Right
   )
  ]
  links [{dst="@OutL", src=AudioMixerAudioMixerStereo2_0.Out L}, {dst="@OutR", src=AudioMixerAudioMixerStereo2_0.Out R}]
 }
 Mobile {
  nodes [
   InitVariable.Trigger()
   n1 = Array.Num:WaveAsset:Array(Array=^Sounds)
   n2 = VariableAccessor.Trigger(Variable=VariableMutator.Trigger(
     Variable=InitVariable.Trigger().Variable
     Value=TriggerCompare.Clamp:Bool(Compare=^OnGunshot, A=^Weapon.IsSuppressorEnabled) {also {defaults {B="<MetaSoundBuilderResult.SUCCEEDED: 0>"}}}.True
    ).Variable)
   ArrayRandomGet_0 = Array.Random Get:WaveAsset:Array(
    In Array=^Sounds
    No Repeats=Max.Max:Int32(A=UE.Subtract:Int32(PrimaryOperand=@n1.Num) {also {defaults {AdditionalOperands=2}}}.Out) {also {defaults {B=1}}}.Value
    Next=@n2.Value
   ) {also {defaults {Enable Shared State="<MetaSoundBuilderResult.SUCCEEDED: 0>", Shared State Behavior=2, Weights []}}}
   n3 = UE.RandomFloat(Max=^Random.Volume.Max, Min=^Random.Volume.Min, Next=UE.Reroute(Value=@ArrayRandomGet_0.On Next).Value)
   FrequencyMultipliertoSemitone_0 = UE.Frequency Multiplier to Semitone:Float(Frequency Multiplier=UE.RandomFloat(Max=^Random.Pitch.Max, Min=^Random.Pitch.Min, Next=VariableAccessor.Trigger(Variable=@n2.Variable).Value).Value)
   WavePlayer_0 = UE.Wave Player:Stereo(Play=@ArrayRandomGet_0.On Next, Wave Asset=@ArrayRandomGet_0.Value, Pitch Shift=@FrequencyMultipliertoSemitone_0.Semitones)
   Multiply_0 = UE.Multiply:Audio by Float(AdditionalOperands=@n3.Value, PrimaryOperand=@WavePlayer_0.Out Left)
   Multiply_1 = UE.Multiply:Audio by Float(AdditionalOperands=@n3.Value, PrimaryOperand=@WavePlayer_0.Out Right)
  ]
  links [{dst="@OutL", src=Multiply_0.Out}, {dst="@OutR", src=Multiply_1.Out}]
 }
]
