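# MetaSound patch that plays third-person suppressed weapon gunshots from random close and medium wave layers with volume variation, firing OnFinished immediately when the suppressor is disabled.
# Default page alternates two players per layer unless single-voice is set, sending summed close to Close.L and summed medium to Close.R.
# Trigger Controls let only the latest close voice signal completion.
# Mobile page uses one player per layer, randomizes pitch and mixes both into the Close stereo pair.
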MetaSoundPatch /FortMetaSoundWeapon/Items/Metasounds/Subgraphs/MSP_Weapon_Fire_Suppressor_3p
inputs [
 bIsSuppressorEnabled {type=Bool}
 bUseSingleVoice {type=Bool}
 OnGunshot {type=Trigger}
 Random.Pitch {type=Float}
 Random.Volume.Max {type=Float}
 Random.Volume.Min {type=Float}
 Sounds.Close {type=WaveAsset:Array}
 Sounds.Medium {type=WaveAsset:Array}
 Volume.Close {type=Float}
 Volume.Medium {type=Float}
]
outputs [
 Close.L {type=Audio}
 Close.R {type=Audio}
 Medium.L {type=Audio}
 OnFinished {type=Trigger}
]
pages [
 Default {
  is_build_graph=true
  input_defaults {
   OnGunshot=null
   Random.Pitch=0
   Random.Volume.Max=1.05
   Random.Volume.Min=0.95
   Sounds.Close=null
   Volume.Close=0
   Volume.Medium=0
   bIsSuppressorEnabled=null
   bUseSingleVoice="<MetaSoundBuilderResult.SUCCEEDED: 0>"
   Sounds.Medium []
  }
  nodes [
   n1 = TriggerCompare.Clamp:Bool(Compare=^OnGunshot, A=^bIsSuppressorEnabled) {also {defaults {B="<MetaSoundBuilderResult.SUCCEEDED: 0>"}}}
   n2 = TriggerCompare.Clamp:Bool(A=^bUseSingleVoice, Compare=@n1.True) {also {defaults {B="<MetaSoundBuilderResult.SUCCEEDED: 0>"}}}
   TriggerSequence2_0 = "UE.Trigger Sequence (2)"(In=@n2.False)
   TriggerAnyTriggerAny2_0 = "TriggerAny.Trigger Any (2)"(In 0=@n2.True, In 1=@TriggerSequence2_0.Out 0)
   n3 = VariableAccessor.Trigger(Variable=VariableMutator.Trigger(Variable=InitVariable.Trigger().Variable, Value=@TriggerAnyTriggerAny2_0.Out).Variable)
   n4 = VariableAccessor.Trigger(Variable=@n3.Variable)
   n5 = VariableAccessor.Trigger(Variable=@n4.Variable)
   n6 = VariableAccessor.Trigger(Variable=@n5.Variable)
   n7 = UE.RandomFloat(Max=^Random.Volume.Max, Min=^Random.Volume.Min, Next=@n6.Value)
   FrequencyMultipliertoSemitone_0 = UE.Frequency Multiplier to Semitone:Float(Frequency Multiplier=^Random.Pitch)
   n8 = VariableAccessor.Trigger(Variable=@n6.Variable)
   n9 = VariableAccessor.Trigger(Variable=VariableMutator.Trigger(Variable=InitVariable.Trigger().Variable, Value=@TriggerSequence2_0.Out 1).Variable)
   n10 = VariableAccessor.Trigger(Variable=@n9.Variable)
   n11 = VariableAccessor.Trigger(Variable=@n10.Variable)
   n12 = VariableAccessor.Trigger(Variable=@n11.Variable)
   n13 = VariableAccessor.Trigger(Variable=@n12.Variable)
   TriggerAnyTriggerAny2_1 = "TriggerAny.Trigger Any (2)"(In 0=VariableAccessor.Trigger(Variable=@n8.Variable).Value, In 1=VariableAccessor.Trigger(Variable=@n13.Variable).Value)
   ArrayRandomGet_1 = Array.Random Get:WaveAsset:Array(
    In Array=^Sounds.Medium
    No Repeats=UE.Divide:Int32(PrimaryOperand=Array.Num:WaveAsset:Array(Array=^Sounds.Medium).Num).Out
    Next=@TriggerAnyTriggerAny2_1.Out
   ) {also {defaults {Enable Shared State="<MetaSoundBuilderResult.SUCCEEDED: 0>", Weights []}}}
   WavePlayer_0 = UE.Wave Player:Stereo(
    Pitch Shift=UE.Reroute(Value=@FrequencyMultipliertoSemitone_0.Semitones).Value
    Wave Asset=UE.Reroute(Value=@ArrayRandomGet_1.Value).Value
    Play=@n5.Value
   )
   n14 = UE.RandomFloat(Max=^Random.Volume.Max, Min=^Random.Volume.Min, Next=@n12.Value)
   WavePlayer_2 = UE.Wave Player:Stereo(
    Pitch Shift=UE.Reroute(Value=@FrequencyMultipliertoSemitone_0.Semitones).Value
    Wave Asset=UE.Reroute(Value=@ArrayRandomGet_1.Value).Value
    Play=@n11.Value
   )
   AudioMixerAudioMixerStereo2_1 = "AudioMixer.Audio Mixer (Stereo, 2)"(
    Gain 0=^Volume.Medium
    Gain 1=^Volume.Medium
    In 0 L=UE.Multiply:Audio by Float(AdditionalOperands=@n7.Value, PrimaryOperand=@WavePlayer_0.Out Left).Out
    In 1 L=UE.Multiply:Audio by Float(AdditionalOperands=@n14.Value, PrimaryOperand=@WavePlayer_2.Out Left).Out
    In 0 R=UE.Multiply:Audio by Float(AdditionalOperands=@n7.Value, PrimaryOperand=@WavePlayer_0.Out Right).Out
    In 1 R=UE.Multiply:Audio by Float(AdditionalOperands=@n14.Value, PrimaryOperand=@WavePlayer_2.Out Right).Out
   )
   Add_0 = UE.Add:Audio(PrimaryOperand=@AudioMixerAudioMixerStereo2_1.Out L, AdditionalOperands=@AudioMixerAudioMixerStereo2_1.Out R)
   n15 = UE.RandomFloat(Max=^Random.Volume.Max, Min=^Random.Volume.Min, Next=@n3.Value)
   TriggerAnyTriggerAny2_4 = "TriggerAny.Trigger Any (2)"(In 0=@n8.Value, In 1=@n13.Value)
   ArrayRandomGet_0 = Array.Random Get:WaveAsset:Array(
    In Array=^Sounds.Close
    No Repeats=UE.Divide:Int32(PrimaryOperand=Array.Num:WaveAsset:Array(Array=^Sounds.Close).Num).Out
    Next=@TriggerAnyTriggerAny2_4.Out
   ) {also {defaults {Enable Shared State="<MetaSoundBuilderResult.SUCCEEDED: 0>", Weights []}}}
   WavePlayer_1 = UE.Wave Player:Stereo(
    Wave Asset=UE.Reroute(Value=@ArrayRandomGet_0.Value).Value
    Pitch Shift=UE.Reroute(Value=@FrequencyMultipliertoSemitone_0.Semitones).Value
    Play=@n4.Value
   )
   n16 = UE.RandomFloat(Max=^Random.Volume.Max, Min=^Random.Volume.Min, Next=@n9.Value)
   WavePlayer_3 = UE.Wave Player:Stereo(
    Wave Asset=UE.Reroute(Value=@ArrayRandomGet_0.Value).Value
    Pitch Shift=UE.Reroute(Value=@FrequencyMultipliertoSemitone_0.Semitones).Value
    Play=@n10.Value
   )
   AudioMixerAudioMixerStereo2_0 = "AudioMixer.Audio Mixer (Stereo, 2)"(
    Gain 0=^Volume.Close
    Gain 1=^Volume.Close
    In 0 L=UE.Multiply:Audio by Float(AdditionalOperands=@n15.Value, PrimaryOperand=@WavePlayer_1.Out Left).Out
    In 1 R=UE.Multiply:Audio by Float(AdditionalOperands=@n16.Value, PrimaryOperand=@WavePlayer_3.Out Right).Out
    In 0 R=UE.Multiply:Audio by Float(AdditionalOperands=@n16.Value, PrimaryOperand=@WavePlayer_3.Out Left).Out
    In 1 L=UE.Multiply:Audio by Float(AdditionalOperands=@n15.Value, PrimaryOperand=@WavePlayer_1.Out Right).Out
   )
   Add_1 = UE.Add:Audio(PrimaryOperand=@AudioMixerAudioMixerStereo2_0.Out L, AdditionalOperands=@AudioMixerAudioMixerStereo2_0.Out R)
   TriggerControl_0 = UE.Trigger Control(Close=@WavePlayer_1.On Play, Trigger In=@WavePlayer_3.On Finished, Open=@WavePlayer_3.On Play)
   TriggerControl_1 = UE.Trigger Control(Trigger In=@WavePlayer_1.On Finished, Open=@WavePlayer_1.On Play, Close=@WavePlayer_3.On Play)
   TriggerAnyTriggerAny2_2 = "TriggerAny.Trigger Any (2)"(In 1=@TriggerControl_0.Trigger Out, In 0=@TriggerControl_1.Trigger Out)
   TriggerAnyTriggerAny2_3 = "TriggerAny.Trigger Any (2)"(In 0=UE.Reroute(Value=@n1.False).Value, In 1=@TriggerAnyTriggerAny2_2.Out)
  ]
  links [{dst="@Close.R", src=Add_0.Out}, {dst="@Close.L", src=Add_1.Out}, {dst="@OnFinished", src=TriggerAnyTriggerAny2_3.Out}]
 }
 Mobile {
  nodes [
   n1 = TriggerCompare.Clamp:Bool(Compare=^OnGunshot, A=^bIsSuppressorEnabled) {also {defaults {B="<MetaSoundBuilderResult.SUCCEEDED: 0>"}}}
   n2 = VariableAccessor.Trigger(Variable=VariableMutator.Trigger(Variable=InitVariable.Trigger().Variable, Value=@n1.True).Variable)
   n3 = VariableAccessor.Trigger(Variable=@n2.Variable)
   n4 = VariableAccessor.Trigger(Variable=@n3.Variable)
   n5 = VariableAccessor.Trigger(Variable=@n4.Variable)
   n6 = UE.RandomFloat(Max=^Random.Volume.Max, Min=^Random.Volume.Min, Next=@n5.Value)
   n7 = VariableAccessor.Trigger(Variable=@n5.Variable)
   n8 = VariableAccessor.Trigger(Variable=@n7.Variable)
   FrequencyMultipliertoSemitone_0 = UE.Frequency Multiplier to Semitone:Float(Frequency Multiplier=UE.RandomFloat(Min=^Random.Pitch, Next=@n8.Value).Value)
   ArrayRandomGet_1 = Array.Random Get:WaveAsset:Array(
    In Array=^Sounds.Medium
    No Repeats=UE.Divide:Int32(PrimaryOperand=Array.Num:WaveAsset:Array(Array=^Sounds.Medium).Num).Out
    Next=VariableAccessor.Trigger(Variable=@n8.Variable).Value
   ) {also {defaults {Enable Shared State="<MetaSoundBuilderResult.SUCCEEDED: 0>", Weights []}}}
   WavePlayer_0 = UE.Wave Player:Stereo(
    Pitch Shift=UE.Reroute(Value=@FrequencyMultipliertoSemitone_0.Semitones).Value
    Wave Asset=UE.Reroute(Value=@ArrayRandomGet_1.Value).Value
    Play=@n4.Value
   )
   n9 = UE.RandomFloat(Max=^Random.Volume.Max, Min=^Random.Volume.Min, Next=@n2.Value)
   ArrayRandomGet_0 = Array.Random Get:WaveAsset:Array(In Array=^Sounds.Close, No Repeats=UE.Divide:Int32(PrimaryOperand=Array.Num:WaveAsset:Array(Array=^Sounds.Close).Num).Out, Next=@n7.Value) {also {defaults {Enable Shared State="<MetaSoundBuilderResult.SUCCEEDED: 0>", Weights []}}}
   WavePlayer_1 = UE.Wave Player:Stereo(
    Pitch Shift=UE.Reroute(Value=@FrequencyMultipliertoSemitone_0.Semitones).Value
    Wave Asset=UE.Reroute(Value=@ArrayRandomGet_0.Value).Value
    Play=@n3.Value
   )
   AudioMixerAudioMixerStereo2_0 = "AudioMixer.Audio Mixer (Stereo, 2)"(
    Gain 0=^Volume.Close
    Gain 1=^Volume.Medium
    In 1 L=UE.Multiply:Audio by Float(AdditionalOperands=@n6.Value, PrimaryOperand=@WavePlayer_0.Out Left).Out
    In 0 L=UE.Multiply:Audio by Float(AdditionalOperands=@n9.Value, PrimaryOperand=@WavePlayer_1.Out Left).Out
    In 1 R=UE.Multiply:Audio by Float(AdditionalOperands=@n6.Value, PrimaryOperand=@WavePlayer_0.Out Right).Out
    In 0 R=UE.Multiply:Audio by Float(AdditionalOperands=@n9.Value, PrimaryOperand=@WavePlayer_1.Out Right).Out
   )
   InitVariable.Trigger()
   TriggerAccumulatorTriggerAccumulate2_0 = "TriggerAccumulator.Trigger Accumulate (2)"(In 1=@WavePlayer_0.On Finished, In 0=@WavePlayer_1.On Finished)
   TriggerAnyTriggerAny2_0 = "TriggerAny.Trigger Any (2)"(In 1=@TriggerAccumulatorTriggerAccumulate2_0.Out, In 0=@n1.False)
  ]
  links [
   {dst="@Close.L", src=AudioMixerAudioMixerStereo2_0.Out L}
   {dst="@Close.R", src=AudioMixerAudioMixerStereo2_0.Out R}
   {dst="@OnFinished", src=TriggerAnyTriggerAny2_0.Out}
  ]
 }
]
